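# Empty SoundCue named as the Jaeger preset for Sourdough vehicle engine audio.
# It has no root or sound nodes, so it plays nothing by itself and carries only playback settings.
# These settings assign the shared DM engine audio attenuation, the in-game SFX sound class and the Fortnite SFX submix, plus priority and maximum distance.
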SoundCue /VehicleAudio/EngineAudio/Sourdough/Jaeger_Preset_SC
root=null
nodes []
properties {
 attenuationSettings=/VehicleAudio/Attenuation/DM_EngineAudio_Attn
 maxDistance=15150
 priority=90
 soundClassObject=&Fort_SFX_InGame
 soundSubmixObject=&Fortnite_SFX
 volumeMultiplier=1
}
